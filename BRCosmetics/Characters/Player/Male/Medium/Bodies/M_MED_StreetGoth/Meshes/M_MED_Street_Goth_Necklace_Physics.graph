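# Physics asset for the necklace on the Street Goth medium male player body cosmetic.
# It defines two single-capsule bodies, a larger one on the dyn_necklace bone and a smaller one on spine_05.
# One constraint joins the necklace to the spine with moderate swing limits and a very tight twist limit, restricting how far it can move relative to the upper torso.
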PhysicsAsset /BRCosmetics/Characters/Player/Male/Medium/Bodies/M_MED_StreetGoth/Meshes/M_MED_Street_Goth_Necklace_Physics
bodies {
 dyn_necklace [{shapeType=Capsule, length=18.4186, radius=8.8681, center=<16.3437 0.7116 0.0037>, rotation=<rot 0 -89.5082 -89.3354>}]
 spine_05 [{shapeType=Capsule, length=8.7088, radius=4.1889, center=<8.8268 0.4536 0.2207>, rotation=<rot -0.0921 -90 -0.3411>}]
}
constraints [{bone1Name=dyn_necklace, bone2Name=spine_05, swing1LimitDegrees=25, swing2LimitDegrees=20, twistLimitDegrees=3}]
