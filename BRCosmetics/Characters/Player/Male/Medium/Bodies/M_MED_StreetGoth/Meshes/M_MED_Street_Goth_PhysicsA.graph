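# Physics asset for the StreetGoth medium male body mesh, covering the legs and their dynamic front, outer and rear leg bones, likely for secondary motion.
# Capsules cover thighs and calves; each thigh links to its calf and three dyn root bones with free angular constraints.
# Front and outer chains end in flat boxes with limited swing and tight twist.
# Same-side front and outer boxes are cross-linked with tighter swing, keeping neighboring panels together.
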PhysicsAsset /BRCosmetics/Characters/Player/Male/Medium/Bodies/M_MED_StreetGoth/Meshes/M_MED_Street_Goth_PhysicsA
bodies {
 calf_l [{shapeType=Capsule, length=19.7607, radius=7.3606, center=<-14.6579 -2.2621 -0.7364>, rotation=<rot 0 87.5055 -89.0356>}]
 calf_r [{shapeType=Capsule, length=20.433, radius=7.1456, center=<14.8389 2.1652 0.7032>, rotation=<rot 0 86.8694 -89.5008>}]
 dyn_frontLeg_l_01 [{shapeType=Capsule, length=1, radius=1}]
 dyn_frontLeg_l_02 [{shapeType=Box, extentX=20, extentY=3, extentZ=17, center=<10.041 0.0001 0>}]
 dyn_frontLeg_r_01 [{shapeType=Capsule, length=1, radius=1}]
 dyn_frontLeg_r_02 [{shapeType=Box, extentX=20, extentY=3, extentZ=17, center=<9.1475 -0.0004 0>}]
 dyn_outerLeg_l_01 [{shapeType=Capsule, length=1, radius=1}]
 dyn_outerLeg_l_02 [{shapeType=Box, extentX=25, extentY=20, extentZ=2, center=<9.35 0 0>}]
 dyn_outerLeg_r_01 [{shapeType=Capsule, length=1, radius=1}]
 dyn_outerLeg_r_02 [{shapeType=Box, extentX=25, extentY=20, extentZ=2, center=<9.35 0 0>}]
 dyn_rearLeg_l_01 [{shapeType=Capsule, length=1, radius=1}]
 dyn_rearLeg_r_01 [{shapeType=Capsule, length=1, radius=1}]
 thigh_l [{shapeType=Capsule, length=31.2186, radius=8.346, center=<-23.8099 -0.2559 0.4763>, rotation=<rot -7.596 93.7677 -89.7839>}]
 thigh_r [{shapeType=Capsule, length=35.831, radius=8.2115, center=<26.6234 0.3274 -0.3566>, rotation=<rot -0.3321 -86.1603 -89.5418>}]
}
constraints [
 {bone1Name=dyn_outerLeg_r_02, bone2Name=dyn_outerLeg_r_01, swing1LimitDegrees=45, swing2LimitDegrees=45, twistLimitDegrees=5}
 {bone1Name=dyn_frontLeg_r_02, bone2Name=dyn_frontLeg_r_01, swing1LimitDegrees=45, swing2LimitDegrees=45, twistLimitDegrees=5}
 {bone1Name=dyn_frontLeg_l_02, bone2Name=dyn_frontLeg_l_01, swing1LimitDegrees=45, swing2LimitDegrees=45, twistLimitDegrees=5}
 {bone1Name=dyn_outerLeg_l_02, bone2Name=dyn_outerLeg_l_01, swing1LimitDegrees=45, swing2LimitDegrees=45, twistLimitDegrees=5}
 {
  bone1Name=dyn_rearLeg_l_01
  bone2Name=thigh_l
  swing1LimitDegrees=45
  swing1Motion=Free
  swing2LimitDegrees=45
  swing2Motion=Free
  twistLimitDegrees=45
  twistMotion=Free
 }
 {
  bone1Name=dyn_outerLeg_l_01
  bone2Name=thigh_l
  swing1LimitDegrees=45
  swing1Motion=Free
  swing2LimitDegrees=45
  swing2Motion=Free
  twistLimitDegrees=45
  twistMotion=Free
 }
 {
  bone1Name=dyn_frontLeg_l_01
  bone2Name=thigh_l
  swing1LimitDegrees=45
  swing1Motion=Free
  swing2LimitDegrees=45
  swing2Motion=Free
  twistLimitDegrees=45
  twistMotion=Free
 }
 {
  bone1Name=calf_l
  bone2Name=thigh_l
  swing1LimitDegrees=45
  swing1Motion=Free
  swing2LimitDegrees=45
  swing2Motion=Free
  twistLimitDegrees=45
  twistMotion=Free
 }
 {
  bone1Name=dyn_frontLeg_r_01
  bone2Name=thigh_r
  swing1LimitDegrees=45
  swing1Motion=Free
  swing2LimitDegrees=45
  swing2Motion=Free
  twistLimitDegrees=45
  twistMotion=Free
 }
 {
  bone1Name=dyn_rearLeg_r_01
  bone2Name=thigh_r
  swing1LimitDegrees=45
  swing1Motion=Free
  swing2LimitDegrees=45
  swing2Motion=Free
  twistLimitDegrees=45
  twistMotion=Free
 }
 {
  bone1Name=dyn_outerLeg_r_01
  bone2Name=thigh_r
  swing1LimitDegrees=45
  swing1Motion=Free
  swing2LimitDegrees=45
  swing2Motion=Free
  twistLimitDegrees=45
  twistMotion=Free
 }
 {
  bone1Name=calf_r
  bone2Name=thigh_r
  swing1LimitDegrees=45
  swing1Motion=Free
  swing2LimitDegrees=45
  swing2Motion=Free
  twistLimitDegrees=45
  twistMotion=Free
 }
 {bone1Name=dyn_outerLeg_r_02, bone2Name=dyn_frontLeg_r_02, swing1LimitDegrees=20, swing2LimitDegrees=20, twistLimitDegrees=10}
 {bone1Name=dyn_outerLeg_l_02, bone2Name=dyn_frontLeg_l_02, swing1LimitDegrees=20, swing2LimitDegrees=20, twistLimitDegrees=10}
]
