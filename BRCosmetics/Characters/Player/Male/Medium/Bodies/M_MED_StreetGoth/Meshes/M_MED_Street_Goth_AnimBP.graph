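# Post-copy AnimBP for the StreetGoth medium male body, adding secondary motion to dynamic leg bones and the necklace.
# Copies the attached parent's pose, then Flap nodes scale front, outer and rear leg bones from calf rotation and rotate them from thighs.
# AnimDynamics and two RigidBody passes add physics, interleaved with additive bone offsets.
# SpineRotLeft, SpeedAlpha and RearCrouchAlpha blend corrective offsets; none are set here, and crouch corrections affect only right-side bones.
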AnimBlueprint /BRCosmetics/Characters/Player/Male/Medium/Bodies/M_MED_StreetGoth/Meshes/M_MED_Street_Goth_AnimBP
target_skeleton=&/M_MED_Street_Goth_Skeleton
graphs [
 AnimGraph {nodes [
   n1 = Flap Node - Bone: dyn_rearLeg_r_01!(
    ComponentPose:Component Space Pose Link Structure=Local To Component(LocalPose=Copy Pose From Mesh() {bCopyCurves=true, bUseAttachedParent=true})
    Alpha:float=1
   ) -> (Pose:Component Space Pose Link Structure) {
    is=title FortAnimGraphNode_Flap
    targetComponent=ScaleX
    alphaBoolBlend {blendOption=Linear}
    sourceDrives [{
      bUseRange=true
      drivingCurve=null
      multiplier=1
      rangeMax=0
      rangeMin=-3
      remappedMax=0
      remappedMin=-1.7
      sourceComponent=RotationZ
      useQuaternion=true
      sourceBone {boneName=calf_r}
     }]
    targetBone {boneName=dyn_rearLeg_r_01}
   }
   n2 = Flap Node - Bone: dyn_outerLeg_r_01!(ComponentPose:Component Space Pose Link Structure=@n1, Alpha:float=1) -> (Pose:Component Space Pose Link Structure) {
    is=title FortAnimGraphNode_Flap
    targetComponent=ScaleX
    alphaBoolBlend {blendOption=Linear}
    sourceDrives [{
      bUseRange=true
      drivingCurve=null
      multiplier=1
      rangeMax=0
      rangeMin=-2.094
      remappedMax=0
      remappedMin=-0.9
      sourceComponent=RotationZ
      useQuaternion=true
      sourceBone {boneName=calf_r}
     }]
    targetBone {boneName=dyn_outerLeg_r_01}
   }
   n3 = Flap Node - Bone: dyn_frontLeg_r_01!(ComponentPose:Component Space Pose Link Structure=@n2, Alpha:float=1) -> (Pose:Component Space Pose Link Structure) {
    is=title FortAnimGraphNode_Flap
    targetComponent=ScaleX
    alphaBoolBlend {blendOption=Linear}
    sourceDrives [{
      bUseRange=true
      drivingCurve=null
      multiplier=1
      rangeMax=0
      rangeMin=-1.2
      remappedMax=0
      remappedMin=-0.6
      sourceComponent=RotationZ
      useQuaternion=true
      sourceBone {boneName=calf_r}
     }]
    targetBone {boneName=dyn_frontLeg_r_01}
   }
   n4 = Flap Node - Bone: dyn_rearLeg_l_01!(ComponentPose:Component Space Pose Link Structure=@n3, Alpha:float=1) -> (Pose:Component Space Pose Link Structure) {
    is=title FortAnimGraphNode_Flap
    targetComponent=ScaleX
    alphaBoolBlend {blendOption=Linear}
    sourceDrives [{
      bUseRange=true
      drivingCurve=null
      multiplier=1
      rangeMax=0
      rangeMin=-3
      remappedMax=0
      remappedMin=-1.7
      sourceComponent=RotationZ
      useQuaternion=true
      sourceBone {boneName=calf_l}
     }]
    targetBone {boneName=dyn_rearLeg_l_01}
   }
   n5 = Flap Node - Bone: dyn_outerLeg_l_01!(ComponentPose:Component Space Pose Link Structure=@n4, Alpha:float=1) -> (Pose:Component Space Pose Link Structure) {
    is=title FortAnimGraphNode_Flap
    targetComponent=ScaleX
    alphaBoolBlend {blendOption=Linear}
    sourceDrives [{
      bUseRange=true
      drivingCurve=null
      multiplier=1
      rangeMax=0
      rangeMin=-2.094
      remappedMax=0
      remappedMin=-0.9
      sourceComponent=RotationZ
      useQuaternion=true
      sourceBone {boneName=calf_l}
     }]
    targetBone {boneName=dyn_outerLeg_l_01}
   }
   n6 = Flap Node - Bone: dyn_frontLeg_l_01!(ComponentPose:Component Space Pose Link Structure=@n5, Alpha:float=1) -> (Pose:Component Space Pose Link Structure) {
    is=title FortAnimGraphNode_Flap
    targetComponent=ScaleX
    alphaBoolBlend {blendOption=Linear}
    sourceDrives [{
      bUseRange=true
      drivingCurve=null
      multiplier=1
      rangeMax=0
      rangeMin=-1.2
      remappedMax=0
      remappedMin=-0.6
      sourceComponent=RotationZ
      useQuaternion=true
      sourceBone {boneName=calf_l}
     }]
    targetBone {boneName=dyn_frontLeg_l_01}
   }
   n7 = Flap Node - Bone: dyn_rearLeg_r_01(ComponentPose=@n6, Alpha=1) {
    targetComponent=RotationZ
    alphaBoolBlend {blendOption=Linear}
    sourceDrives [{
      bUseRange=true
      drivingCurve=null
      multiplier=1
      rangeMax=0.35
      rangeMin=-0.35
      remappedMax=0.35
      remappedMin=-0.35
      sourceComponent=RotationZ
      useQuaternion=true
      sourceBone {boneName=thigh_r}
     }]
    targetBone {boneName=dyn_rearLeg_r_01}
   }
   n8 = Flap Node - Bone: dyn_outerLeg_r_01(ComponentPose=@n7, Alpha=1) {
    targetComponent=RotationZ
    alphaBoolBlend {blendOption=Linear}
    sourceDrives [{
      bUseRange=true
      drivingCurve=null
      multiplier=1
      rangeMax=0.35
      rangeMin=-0.35
      remappedMax=0.35
      remappedMin=-0.35
      sourceComponent=RotationZ
      useQuaternion=true
      sourceBone {boneName=thigh_r}
     }]
    targetBone {boneName=dyn_outerLeg_r_01}
   }
   n9 = Flap Node - Bone: dyn_frontLeg_r_01(ComponentPose=@n8, Alpha=1) {
    targetComponent=RotationZ
    alphaBoolBlend {blendOption=Linear}
    sourceDrives [{
      bUseRange=true
      drivingCurve=null
      multiplier=1
      rangeMax=0.1725
      rangeMin=-0.1725
      remappedMax=0.1725
      remappedMin=-0.1725
      sourceComponent=RotationZ
      useQuaternion=true
      sourceBone {boneName=thigh_r}
     }]
    targetBone {boneName=dyn_frontLeg_r_01}
   }
   n10 = Flap Node - Bone: dyn_rearLeg_l_01(ComponentPose=@n9, Alpha=1) {
    targetComponent=RotationZ
    alphaBoolBlend {blendOption=Linear}
    sourceDrives [{
      bUseRange=true
      drivingCurve=null
      multiplier=1
      rangeMax=0.35
      rangeMin=-0.35
      remappedMax=0.35
      remappedMin=-0.35
      sourceComponent=RotationZ
      useQuaternion=true
      sourceBone {boneName=thigh_l}
     }]
    targetBone {boneName=dyn_rearLeg_l_01}
   }
   n11 = Flap Node - Bone: dyn_outerLeg_l_01(ComponentPose=@n10, Alpha=1) {
    targetComponent=RotationZ
    alphaBoolBlend {blendOption=Linear}
    sourceDrives [{
      bUseRange=true
      drivingCurve=null
      multiplier=1
      rangeMax=0.35
      rangeMin=-0.35
      remappedMax=0.35
      remappedMin=-0.35
      sourceComponent=RotationZ
      useQuaternion=true
      sourceBone {boneName=thigh_l}
     }]
    targetBone {boneName=dyn_outerLeg_l_01}
   }
   n12 = Flap Node - Bone: dyn_frontLeg_l_01(ComponentPose=@n11, Alpha=1) {
    targetComponent=RotationZ
    alphaBoolBlend {blendOption=Linear}
    sourceDrives [{
      bUseRange=true
      drivingCurve=null
      multiplier=1
      rangeMax=0.1725
      rangeMin=-0.1725
      remappedMax=0.1725
      remappedMin=-0.1725
      sourceComponent=RotationZ
      useQuaternion=true
      sourceBone {boneName=thigh_l}
     }]
    targetBone {boneName=dyn_frontLeg_l_01}
   }
   n13 = Anim Dynamics - Bone: dyn_l_mstr(ComponentPose=@n12, Alpha=1, ExternalForce="0,0,0") {
    angularSpringConstant=115
    bAngularSpring=true
    bGravityOverrideInSimSpace=true
    bLinearSpring=true
    gravityScale=0.35
    lODThreshold=1
    linearSpringConstant=115
    simulationSpace=Actor
    windScale=0.5
    alphaBoolBlend {blendOption=Linear}
    boundBone {boneName=dyn_l_mstr}
    physicsBodyDefinitions [{
      sphereCollisionRadius=0
      boundBone {boneName=dyn_l_mstr}
      boxExtents=<30 30 30>
      constraintSetup {
       angularTargetAxis=AxisY
       twistAxis=AxisY
       angularLimitsMax=<7.5 12 7.5>
       angularLimitsMin=<-7.5 -12 -7.5>
       angularTarget=<0 0 0>
       linearAxesMax=<3 0 0>
       linearAxesMin=<-3 0 0>
      }
      localJointOffset=<0 0 0>
     }]
   }
   n14 = Anim Dynamics - Bone: dyn_outerLeg_l_01(ComponentPose=@n13, Alpha=1, ExternalForce="0,0,0") {
    angularSpringConstant=115
    bAngularSpring=true
    bGravityOverrideInSimSpace=true
    bLinearSpring=true
    gravityScale=0.35
    lODThreshold=1
    linearSpringConstant=115
    simulationSpace=RootRelative
    windScale=0.5
    alphaBoolBlend {blendOption=Linear}
    boundBone {boneName=dyn_outerLeg_l_01}
    componentAppliedLinearAccClamp=<10000 10000 10000>
    componentLinearAccScale=<1 1 1>
    componentLinearVelScale=<1 1 1>
    physicsBodyDefinitions [{
      sphereCollisionRadius=0
      boundBone {boneName=dyn_outerLeg_l_01}
      boxExtents=<30 30 30>
      constraintSetup {angularLimitsMax=<4.0605 1 25>, angularLimitsMin=<-10.9395 -43 -25>, angularTarget=<0 0 0>}
      localJointOffset=<15 0 0>
     }]
    relativeSpaceBone {boneName=root}
   }
   n15 = Anim Dynamics - Bone: dyn_r_mstr(ComponentPose=@n14, Alpha=1, ExternalForce="0,0,0") {
    angularSpringConstant=115
    bAngularSpring=true
    bGravityOverrideInSimSpace=true
    bLinearSpring=true
    gravityScale=0.35
    lODThreshold=1
    linearSpringConstant=115
    simulationSpace=Actor
    windScale=0.5
    alphaBoolBlend {blendOption=Linear}
    boundBone {boneName=dyn_r_mstr}
    physicsBodyDefinitions [{
      sphereCollisionRadius=0
      boundBone {boneName=dyn_r_mstr}
      boxExtents=<30 30 30>
      constraintSetup {
       angularTargetAxis=AxisY
       twistAxis=AxisY
       angularLimitsMax=<7.5 12 7.5>
       angularLimitsMin=<-7.5 -12 -7.5>
       angularTarget=<0 0 0>
       linearAxesMax=<3 0 0>
       linearAxesMin=<-3 0 0>
      }
      localJointOffset=<0 0 0>
     }]
   }
   n16 = Anim Dynamics - Bone: dyn_outerLeg_r_01(ComponentPose=@n15, Alpha=1, ExternalForce="0,0,0") {
    angularSpringConstant=115
    bAngularSpring=true
    bGravityOverrideInSimSpace=true
    bLinearSpring=true
    gravityScale=0.35
    lODThreshold=1
    linearSpringConstant=115
    simulationSpace=RootRelative
    windScale=0.5
    alphaBoolBlend {blendOption=Linear}
    boundBone {boneName=dyn_outerLeg_r_01}
    componentAppliedLinearAccClamp=<10000 10000 10000>
    componentLinearAccScale=<1 1 1>
    componentLinearVelScale=<1 1 1>
    physicsBodyDefinitions [{
      sphereCollisionRadius=0
      boundBone {boneName=dyn_outerLeg_r_01}
      boxExtents=<30 30 30>
      constraintSetup {angularLimitsMax=<4.0605 43 10>, angularLimitsMin=<-10.9395 1 -25>, angularTarget=<0 0 0>}
      localJointOffset=<15 0 0>
     }]
    relativeSpaceBone {boneName=root}
   }
   n17 = RigidBody(ComponentPose=@n16) {
    bOverrideWorldGravity=true
    overridePhysicsAsset=&/M_MED_Street_Goth_PhysicsA
    simulationSpace=BaseBoneSpace
    alphaBoolBlend {blendOption=Linear}
    baseBoneRef {boneName=root}
    componentLinearAccScale=<1 1 1>
    componentLinearVelScale=<1 1 1>
    overrideWorldGravity {z=-980}
   }
   n18 = Anim Dynamics - Bone: thigh_twist_01_r(ComponentPose=@n17, Alpha=0.25, ExternalForce="0,0,0") {
    angularSpringConstant=80
    bAngularSpring=true
    bGravityOverrideInSimSpace=true
    bLinearSpring=true
    gravityScale=0.35
    lODThreshold=1
    linearSpringConstant=80
    simulationSpace=Actor
    windScale=0.5
    alphaBoolBlend {blendOption=Linear}
    boundBone {boneName=thigh_twist_01_r}
    physicsBodyDefinitions [{
      sphereCollisionRadius=0
      boundBone {boneName=thigh_twist_01_r}
      boxExtents=<30 30 30>
      constraintSetup {
       angularTargetAxis=AxisY
       twistAxis=AxisY
       angularLimitsMax=<8 8 0>
       angularLimitsMin=<-8 -8 0>
       angularTarget=<0 0 0>
       linearAxesMax=<0 0 2.5>
       linearAxesMin=<0 0 -2.5>
      }
      localJointOffset=<0 0 0>
     }]
   }
   n19 = Anim Dynamics - Bone: thigh_twist_01_l(ComponentPose=@n18, Alpha=0.25, ExternalForce="0,0,0") {
    angularSpringConstant=80
    bAngularSpring=true
    bGravityOverrideInSimSpace=true
    bLinearSpring=true
    gravityScale=0.35
    lODThreshold=1
    linearSpringConstant=80
    simulationSpace=Actor
    windScale=0.5
    alphaBoolBlend {blendOption=Linear}
    boundBone {boneName=thigh_twist_01_l}
    physicsBodyDefinitions [{
      sphereCollisionRadius=0
      boundBone {boneName=thigh_twist_01_l}
      boxExtents=<30 30 30>
      constraintSetup {
       angularTargetAxis=AxisY
       twistAxis=AxisY
       angularLimitsMax=<8 8 0>
       angularLimitsMin=<-8 -8 0>
       angularTarget=<0 0 0>
       linearAxesMax=<0 0 2.5>
       linearAxesMin=<0 0 -2.5>
      }
      localJointOffset=<0 0 0>
     }]
   }
   n20 = "Transform (Modify) Bone - Bone: dyn_l_mstr_leg"(ComponentPose=@n19, Alpha=1, Translation="0,0,0", Rotation="0,0,-50", Scale="1,1,1") {rotationMode=BMM_Additive, alphaBoolBlend {blendOption=Linear}, boneToModify {boneName=dyn_l_mstr_leg}, rotation {roll=-50}}
   n21 = "Transform (Modify) Bone - Bone: dyn_r_mstr_leg"(ComponentPose=@n20, Alpha=1, Translation="0,0,0", Rotation="0.000,0.000,-50", Scale="1,1,1") {rotationMode=BMM_Additive, alphaBoolBlend {blendOption=Linear}, boneToModify {boneName=dyn_r_mstr_leg}, rotation {roll=-50}}
   n22 = "Transform (Modify) Bone - Bone: thigh_twist_01_r"(
    ComponentPose=@n21
    Alpha=Get SpineRotLeft() -> (SpineRotLeft:float)
    Translation="0.289,-1.759,1.145"
    Rotation="2.626,4.512,-20.392"
    Scale="1,1,1"
   ) {
    rotationMode=BMM_Additive
    rotationSpace=BCS_BoneSpace
    translationMode=BMM_Additive
    translationSpace=BCS_BoneSpace
    alphaBoolBlend {blendOption=Linear}
    alphaScaleBiasClamp {bInterpResult=true, interpSpeedDecreasing=3}
    boneToModify {boneName=thigh_twist_01_r}
    rotation=<rot 2.626 4.512 -20.392>
    translation=<0.289 -1.759 1.145>
   }
   n23 = Get SpeedAlpha() -> (SpeedAlpha:float)
   n24 = "Transform (Modify) Bone - Bone: dyn_rearLeg_driver_l"(ComponentPose=@n22, Alpha=@n23, Translation="0,0,0", Rotation="35,0,0", Scale="1,1,1") {
    rotationMode=BMM_Additive
    alphaBoolBlend {blendOption=Linear}
    alphaScaleBiasClamp {bInterpResult=true}
    boneToModify {boneName=dyn_rearLeg_driver_l}
    rotation {pitch=35}
   }
   n25 = "Transform (Modify) Bone - Bone: dyn_rearLeg_driver_r"(ComponentPose=@n24, Alpha=@n23, Translation="0,0,0", Rotation="-35,0,0", Scale="1,1,1") {
    rotationMode=BMM_Additive
    rotationSpace=BCS_BoneSpace
    alphaBoolBlend {blendOption=Linear}
    alphaScaleBiasClamp {bInterpResult=true}
    boneToModify {boneName=dyn_rearLeg_driver_r}
    rotation {pitch=-35}
   }
   n26 = Get RearCrouchAlpha() -> (RearCrouchAlpha:float)
   n27 = "Transform (Modify) Bone - Bone: dyn_rearLeg_r_01"(ComponentPose=@n25, Alpha=@n26, Translation="0,0,0", Rotation="0,-50,0", Scale="0.25,1,1") {
    rotationMode=BMM_Additive
    rotationSpace=BCS_BoneSpace
    scaleMode=BMM_Replace
    scaleSpace=BCS_BoneSpace
    translationMode=BMM_Additive
    translationSpace=BCS_BoneSpace
    alphaBoolBlend {blendOption=Linear}
    alphaScaleBiasClamp {bInterpResult=true}
    boneToModify {boneName=dyn_rearLeg_r_01}
    rotation {yaw=-50}
    scale {x=0.25}
   }
   n28 = "Transform (Modify) Bone - Bone: dyn_outerLeg_r_01"(ComponentPose=@n27, Alpha=@n26, Translation="0,0,0", Rotation="0,-60,0", Scale="0.1,1,1") {
    rotationMode=BMM_Additive
    rotationSpace=BCS_BoneSpace
    scaleMode=BMM_Replace
    scaleSpace=BCS_BoneSpace
    translationMode=BMM_Additive
    translationSpace=BCS_BoneSpace
    alphaBoolBlend {blendOption=Linear}
    alphaScaleBiasClamp {bInterpResult=true}
    boneToModify {boneName=dyn_outerLeg_r_01}
    rotation {yaw=-60}
    scale {x=0.1}
   }
   n29 = RigidBody(ComponentPose=@n28) {
    bOverrideWorldGravity=true
    overridePhysicsAsset=&/M_MED_Street_Goth_Necklace_Physics
    simulationSpace=BaseBoneSpace
    alphaBoolBlend {blendOption=Linear}
    baseBoneRef {boneName=spine_05}
    componentAppliedLinearAccClamp=<500 500 500>
    componentLinearAccScale=<1 1 1>
    componentLinearVelScale=<1 1 1>
    overrideWorldGravity {z=-980}
   }
   Output Pose(Result=Component To Local(ComponentPose=@n29)) {layerGroup=None}
  ]}
 EventGraph {links [], nodes []}
]
variables []
